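# Physics asset for the Famine backpack cosmetic, providing collision bodies and joints for its dangling chain, jaw and bone parts.
# An eight-link capsule chain hangs from the pack box, with small twist limits per link and swing locked below the first link.
# The jaw twists on the pack, while the bone capsule links to jaw, pack and root.
# Root and spine capsules form a torso shape; the spine bodies have no constraints.
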PhysicsAsset /BRCosmetics/Accessories/FORT_Backpacks/Backpack_M_MED_Famine/Meshes/M_MED_Famine_Pack_Physics
bodies {
 dyn_bone [{shapeType=Capsule, length=32.8976, radius=2.0496, center=<-1.6753 -0.9404 0.2449>, rotation=<rot 0 -89.5587 -86.1964>}]
 dyn_chain_01 [{shapeType=Capsule, length=4.027, radius=1.6888, shapeName=dyn_chain_01_capsule, center=<0 0 -2.2311>}]
 dyn_chain_02 [{shapeType=Capsule, length=2.6568, radius=1.4491, shapeName=dyn_chain_02_capsule, center=<0 0 -1.2591>}]
 dyn_chain_03 [{shapeType=Capsule, length=2.4567, radius=1.4491, shapeName=dyn_chain_03_capsule, center=<0 0 -1.2591>}]
 dyn_chain_04 [{shapeType=Capsule, length=2.5367, radius=1.4491, shapeName=dyn_chain_04_capsule, center=<0 0 -1.2591>}]
 dyn_chain_05 [{shapeType=Capsule, length=2.4568, radius=1.4491, shapeName=dyn_chain_05_capsule, center=<0 0 -1.2591>}]
 dyn_chain_06 [{shapeType=Capsule, length=2.2596, radius=1.4491, shapeName=dyn_chain_06_capsule, center=<0 0 -0.9988>}]
 dyn_chain_07 [{shapeType=Capsule, length=2.9768, radius=1.4491, shapeName=dyn_chain_07_capsule, center=<0 0 -1.2591>}]
 dyn_chain_08 [{shapeType=Capsule, length=4.5188, radius=2.7367, shapeName=dyn_chain_08_capsule, center=<0 0 -2.9003>}]
 dyn_jaw [{shapeType=Box, extentX=25, extentY=15.5, extentZ=1.5, center=<0 -3.7007 -5.9547>, rotation=<rot 0 0 -41.3623>}]
 pack [{shapeType=Box, extentX=7.5, extentY=4, extentZ=23, center=<0.0001 4.38 0>}]
 root [{
   shapeType=Capsule
   length=16.2501
   radius=11.4393
   shapeName=root_capsule
   center=<-0.0138 12.93 -7.3573>
   rotation=<rot -89.0564 -179.9999 179.9999>
  }]
 spine_01 [{shapeType=Capsule, length=24.3888, radius=12.0223, shapeName=spine_01_capsule, center=<0 1.3734 0>, rotation=<rot 0.0752 180 180>}]
 spine_02 [{shapeType=Capsule, length=14.2662, radius=12.9223, shapeName=spine_02_capsule, center=<0 1.3734 0>, rotation=<rot 0.0752 180 180>}]
 spine_03 [{shapeType=Capsule, length=13.6911, radius=11.34, shapeName=spine_03_capsule}]
}
constraints [
 {bone1Name=dyn_jaw, bone2Name=pack, swing1LimitDegrees=0, swing2LimitDegrees=0, twistLimitDegrees=60}
 {bone1Name=dyn_chain_01, bone2Name=pack, swing1LimitDegrees=0.01, swing2LimitDegrees=0.01, twistLimitDegrees=10}
 {
  bone1Name=dyn_chain_02
  bone2Name=dyn_chain_01
  swing1LimitDegrees=0.01
  swing1Motion=Locked
  swing2LimitDegrees=0.01
  swing2Motion=Locked
  twistLimitDegrees=10
 }
 {
  bone1Name=dyn_chain_03
  bone2Name=dyn_chain_02
  swing1LimitDegrees=2
  swing1Motion=Locked
  swing2LimitDegrees=2
  swing2Motion=Locked
  twistLimitDegrees=10
 }
 {
  bone1Name=dyn_chain_04
  bone2Name=dyn_chain_03
  swing1LimitDegrees=2
  swing1Motion=Locked
  swing2LimitDegrees=2
  swing2Motion=Locked
  twistLimitDegrees=10
 }
 {
  bone1Name=dyn_chain_05
  bone2Name=dyn_chain_04
  swing1LimitDegrees=2
  swing1Motion=Locked
  swing2LimitDegrees=2
  swing2Motion=Locked
  twistLimitDegrees=10
 }
 {
  bone1Name=dyn_chain_06
  bone2Name=dyn_chain_05
  swing1LimitDegrees=2
  swing1Motion=Locked
  swing2LimitDegrees=2
  swing2Motion=Locked
  twistLimitDegrees=10
 }
 {
  bone1Name=dyn_chain_07
  bone2Name=dyn_chain_06
  swing1LimitDegrees=2
  swing1Motion=Locked
  swing2LimitDegrees=2
  swing2Motion=Locked
  twistLimitDegrees=10
 }
 {
  bone1Name=dyn_chain_08
  bone2Name=dyn_chain_07
  swing1LimitDegrees=2
  swing1Motion=Locked
  swing2LimitDegrees=2
  swing2Motion=Locked
  twistLimitDegrees=10
 }
 {bone1Name=dyn_jaw, bone2Name=dyn_bone, swing1LimitDegrees=8, swing2LimitDegrees=8, twistLimitDegrees=20}
 {bone1Name=pack, bone2Name=dyn_bone, swing1LimitDegrees=8, swing2LimitDegrees=8, twistLimitDegrees=15}
 {bone1Name=pack, bone2Name=root, swing1LimitDegrees=45, swing2LimitDegrees=45, twistLimitDegrees=45}
 {bone1Name=dyn_bone, bone2Name=root, swing1LimitDegrees=45, swing2LimitDegrees=45, twistLimitDegrees=45}
]
