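# Animation Blueprint for the Famine backpack cosmetic that follows the wearer's spine and adds secondary physics motion.
# Each update, the event graph caches world transforms of spine_01 through spine_03 from the attach parent.
# The AnimGraph offsets the pack bone from reference pose, then applies those transforms, with spine_03's (spine04 variables) overwriting spine_02.
# A RigidBody node then simulates the pack's physics asset relative to the pack bone with overridden gravity.
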AnimBlueprint /BRCosmetics/Accessories/FORT_Backpacks/Backpack_M_MED_Famine/Meshes/M_MED_Famine_Pack_AnimBP
target_skeleton=&/M_MED_Famine_Pack_Skeleton
graphs [
 AnimGraph {nodes [
   n1 = "Transform (Modify) Bone - Bone: pack"(ComponentPose=Local To Component(LocalPose=Local Space Ref Pose()), Translation="0,-1.8,0", Rotation="0,0,7") {
    rotationMode=BMM_Additive
    rotationSpace=BCS_BoneSpace
    translationMode=BMM_Additive
    translationSpace=BCS_BoneSpace
    alphaBoolBlend {blendOption=Linear}
    boneToModify {boneName=pack}
    rotation {roll=7}
    translation {y=-1.8}
   }
   n2 = "Transform (Modify) Bone - Bone: spine_01"(
    ComponentPose=@n1
    Alpha=1
    Translation=Get spine01_tran() -> (spine01_tran:Vector)
    Rotation=Get spine01_rot() -> (spine01_rot:Rotator)
    Scale="1,1,1"
   ) {
    rotationMode=BMM_Replace
    rotationSpace=BCS_WorldSpace
    translationMode=BMM_Replace
    translationSpace=BCS_WorldSpace
    alphaBoolBlend {blendOption=Linear}
    boneToModify {boneName=spine_01}
    rotation {roll=7}
    translation {y=-1.8}
   }
   n3 = "Transform (Modify) Bone - Bone: spine_02"(
    ComponentPose=@n2
    Alpha=1
    Translation=Get spine02_tran() -> (spine02_tran:Vector)
    Rotation=Get spine02_rot() -> (spine02_rot:Rotator)
    Scale="1,1,1"
   ) {
    rotationMode=BMM_Replace
    rotationSpace=BCS_WorldSpace
    translationMode=BMM_Replace
    translationSpace=BCS_WorldSpace
    alphaBoolBlend {blendOption=Linear}
    boneToModify {boneName=spine_02}
    rotation {roll=7}
    translation {y=-1.8}
   }
   n4 = "Transform (Modify) Bone - Bone: spine_02"(
    ComponentPose=@n3
    Alpha=1
    Translation=Get spine04_tran() -> (spine04_tran:Vector)
    Rotation=Get spine04_rot() -> (spine04_rot:Rotator)
    Scale="1,1,1"
   ) {
    rotationMode=BMM_Replace
    rotationSpace=BCS_WorldSpace
    translationMode=BMM_Replace
    translationSpace=BCS_WorldSpace
    alphaBoolBlend {blendOption=Linear}
    boneToModify {boneName=spine_02}
    rotation {roll=7}
    translation {y=-1.8}
   }
   n5 = RigidBody(ComponentPose=@n4) {
    bClampLinearTranslationLimitToRefPose=true
    bOverrideWorldGravity=true
    overridePhysicsAsset=&/M_MED_Famine_Pack_Physics
    simulationSpace=BaseBoneSpace
    worldSpaceMinimumScale=1
    alphaBoolBlend {blendOption=Linear}
    baseBoneRef {boneName=pack}
    componentAppliedLinearAccClamp=<700 700 700>
    componentLinearAccScale=<0.35 0.45 0.45>
    componentLinearVelScale=<0.3 0.3 0.3>
    overrideWorldGravity {z=-980}
   }
   Output Pose(Result=Component To Local(ComponentPose=@n5)) {layerGroup=None}
  ]}
 EventGraph {nodes [
   Event BlueprintUpdateAnimation()
   n1 = GetPartSkeletalMeshComponent()
   > Is Valid(InputObject=@n1)
   n2 = GetAttachParent!(self:Scene Component Object Reference=@n1) -> (ReturnValue:Scene Component Object Reference)
   > Is Valid(InputObject=@n2)
   n3 = BreakTransform(InTransform=GetSocketTransform(self=@n2, InSocketName=spine_01, TransformSpace=RTS_World))
   > Set spine01_tran(spine01_tran:Vector=@n3) -> (Output_Get:Vector)
   > Set spine01_rot(spine01_rot:Rotator=@n3.Rotation) -> (Output_Get:Rotator)
   n4 = BreakTransform(InTransform=GetSocketTransform(self=@n2, InSocketName=spine_02, TransformSpace=RTS_World))
   > Set spine02_tran(spine02_tran:Vector=@n4) -> (Output_Get:Vector)
   > Set spine02_rot(spine02_rot:Rotator=@n4.Rotation) -> (Output_Get:Rotator)
   n5 = BreakTransform(InTransform=GetSocketTransform(self=@n2, InSocketName=spine_03, TransformSpace=RTS_World))
   > Set spine04_tran(spine04_tran:Vector=@n5) -> (Output_Get:Vector)
   > Set spine04_rot(spine04_rot:Rotator=@n5.Rotation) -> (Output_Get:Rotator)
   TryGetPawnOwner()
  ]}
]
variables [
 spine01_rot {type=Rotator}
 spine01_tran {type=Vector}
 spine02_rot {type=Rotator}
 spine02_tran {type=Vector}
 spine04_rot {type=Rotator}
 spine04_tran {type=Vector}
]
